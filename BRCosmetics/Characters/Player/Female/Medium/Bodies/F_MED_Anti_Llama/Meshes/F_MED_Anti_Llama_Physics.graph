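# Physics asset for the Anti Llama female medium player body, covering secondary motion of its belt, thigh gun and torn pants pieces.
# Capsules on the pelvis and both thighs form the body, linked by free-rotation constraints.
# Pants tear bones attach directly to their thighs with tight angular limits.
# Gun and belt hang from small base bodies, freely jointed to left thigh and pelvis respectively, with narrow limits between base and prop.
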PhysicsAsset /BRCosmetics/Characters/Player/Female/Medium/Bodies/F_MED_Anti_Llama/Meshes/F_MED_Anti_Llama_Physics
bodies {
 dyn_belt [{shapeType=Capsule, length=4.5, radius=3.75, center=<6.8175 0 0>, rotation=<rot 90 0 0>}]
 dyn_belt_base [{shapeType=Capsule, length=0.505, radius=0.505, rotation=<rot 0 -90 0>}]
 dyn_gun [{shapeType=Capsule, length=17.75, radius=3.5, center=<1.4593 0 2.8052>, rotation=<rot 90 0 0>}]
 dyn_gun_base [{shapeType=Capsule, length=1, radius=1}]
 dyn_pants_tear_a_l [{shapeType=Capsule, length=1.25, radius=1.5, center=<2 0 0>, rotation=<rot 90 0 0>}]
 dyn_pants_tear_a_r [{shapeType=Capsule, length=1.25, radius=1.5, center=<2 0 0>, rotation=<rot 90 0 0>}]
 dyn_pants_tear_b_l [{shapeType=Capsule, length=1.25, radius=1.5, center=<2 0 0>, rotation=<rot 90 0 0>}]
 dyn_pants_tear_b_r [{shapeType=Capsule, length=1.25, radius=1.5, center=<2 0 0>, rotation=<rot 90 0 0>}]
 dyn_pants_tear_c_l [{shapeType=Capsule, length=1.25, radius=1.5, center=<2 0 0>, rotation=<rot 90 0 0>}]
 pelvis [{shapeType=Capsule, length=8.25, radius=9.75, center=<-0.5964 -2.1273 0>}]
 thigh_l [{shapeType=Capsule, length=23.5, radius=7.75, center=<-15.2931 0.3063 0.6854>, rotation=<rot 90 20.556 -156.513>}]
 thigh_r [{shapeType=Capsule, length=23.5, radius=7.75, center=<15.2931 -0.3063 -0.6854>, rotation=<rot 90 -90 -87.069>}]
}
constraints [
 {
  bone1Name=thigh_r
  bone2Name=pelvis
  swing1LimitDegrees=45
  swing1Motion=Free
  swing2LimitDegrees=45
  swing2Motion=Free
  twistLimitDegrees=45
  twistMotion=Free
 }
 {bone1Name=dyn_pants_tear_a_l, bone2Name=thigh_l, swing1LimitDegrees=15, swing2LimitDegrees=10, twistLimitDegrees=5}
 {bone1Name=dyn_pants_tear_c_l, bone2Name=thigh_l, swing1LimitDegrees=15, swing2LimitDegrees=10, twistLimitDegrees=5}
 {bone1Name=dyn_pants_tear_b_l, bone2Name=thigh_l, swing1LimitDegrees=15, swing2LimitDegrees=10, twistLimitDegrees=5}
 {bone1Name=dyn_pants_tear_a_r, bone2Name=thigh_r, swing1LimitDegrees=10, swing2LimitDegrees=5, twistLimitDegrees=5}
 {bone1Name=dyn_pants_tear_b_r, bone2Name=thigh_r, swing1LimitDegrees=15, swing2LimitDegrees=10, twistLimitDegrees=5}
 {
  bone1Name=thigh_l
  bone2Name=pelvis
  swing1LimitDegrees=45
  swing1Motion=Free
  swing2LimitDegrees=45
  swing2Motion=Free
  twistLimitDegrees=45
  twistMotion=Free
 }
 {
  bone1Name=dyn_gun_base
  bone2Name=thigh_l
  swing1LimitDegrees=45
  swing1Motion=Free
  swing2LimitDegrees=45
  swing2Motion=Free
  twistLimitDegrees=45
  twistMotion=Free
 }
 {bone1Name=dyn_gun, bone2Name=dyn_gun_base, swing1LimitDegrees=3, swing2LimitDegrees=5, twistLimitDegrees=5}
 {
  bone1Name=dyn_belt_base
  bone2Name=pelvis
  swing1LimitDegrees=45
  swing1Motion=Free
  swing2LimitDegrees=45
  swing2Motion=Free
  twistLimitDegrees=45
  twistMotion=Free
 }
 {bone1Name=dyn_belt, bone2Name=dyn_belt_base, swing1LimitDegrees=5, swing2LimitDegrees=25, twistLimitDegrees=1}
]
