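# Post-process style AnimBlueprint for the F_MED_Anti_Llama female cosmetic body, adding secondary motion to a pose copied from the parent mesh.
# Event graph derives thigh angles and fall-speed alpha, driving corrective thigh and belt bone offsets.
# Weighted constraints place ammo bones, then pelvis-space RigidBody simulation, collar constraints and a Control Rig finish the pose.
# Collar offset nodes are disconnected; look_up and LOD1 are computed but unread here.
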AnimBlueprint /BRCosmetics/Characters/Player/Female/Medium/Bodies/F_MED_Anti_Llama/Meshes/F_MED_Anti_Llama_AnimBP
target_skeleton=&/F_MED_Anti_Llama_Skeleton
graphs [
 AnimGraph {nodes [
   n1 = "Transform (Modify) Bone - Bone: dyn_collar_bk"(Alpha=1, Translation="0,-0.3,1.2", Rotation="10,0,0", Scale="1,1,1") {
    rotationMode=BMM_Additive
    rotationSpace=BCS_BoneSpace
    translationMode=BMM_Additive
    translationSpace=BCS_BoneSpace
    alphaBoolBlend {blendOption=Linear}
    boneToModify {boneName=dyn_collar_bk}
    rotation {pitch=10}
    translation {y=-0.3, z=1.2}
   }
   n2 = "Transform (Modify) Bone - Bone: dyn_collar_r_mstr"(ComponentPose=@n1, Alpha=1, Translation="0,0.25,0.7", Rotation="0,0,-2", Scale="1,1,1") {
    rotationMode=BMM_Additive
    rotationSpace=BCS_BoneSpace
    translationMode=BMM_Additive
    translationSpace=BCS_BoneSpace
    alphaBoolBlend {blendOption=Linear}
    boneToModify {boneName=dyn_collar_r_mstr}
    rotation {roll=-2}
    translation {y=0.25, z=0.7}
   }
   "Transform (Modify) Bone - Bone: dyn_collar_l_mstr"(ComponentPose=@n2, Alpha=1, Translation="0,0,0.7", Rotation="0,0,2", Scale="1,1,1") {
    rotationMode=BMM_Additive
    rotationSpace=BCS_BoneSpace
    translationMode=BMM_Additive
    translationSpace=BCS_BoneSpace
    alphaBoolBlend {blendOption=Linear}
    boneToModify {boneName=dyn_collar_l_mstr}
    rotation {roll=2}
    translation {z=0.7}
   }
   n3 = "Constraint - thigh_volume_l (R)"(ComponentPose=Local To Component(LocalPose=Copy Pose From Mesh() {bUseAttachedParent=true}), Alpha=0.5, ConstraintWeights_0=1) {
    alphaBoolBlend {blendOption=Linear}
    boneToModify {boneName=thigh_volume_l}
    constraintSetup [{transformType=Rotation, targetBone {boneName=thigh_l}}]
    constraintWeights [1]
   }
   n4 = "Transform (Modify) Bone - Bone: thigh_volume_l"(ComponentPose=@n3, Alpha=Get thigh_rot_l(), Translation="-1,-2,1", Rotation="0,0,0", Scale="1,1,1") {
    translationMode=BMM_Additive
    translationSpace=BCS_BoneSpace
    alphaBoolBlend {blendOption=Linear}
    boneToModify {boneName=thigh_volume_l}
    translation=<-1 -2 1>
   }
   n5 = "Constraint - thigh_volume_r (R)"(ComponentPose=@n4, Alpha=0.5, ConstraintWeights_0=1) {
    alphaBoolBlend {blendOption=Linear}
    boneToModify {boneName=thigh_volume_r}
    constraintSetup [{transformType=Rotation, targetBone {boneName=thigh_r}}]
    constraintWeights [1]
   }
   n6 = "Transform (Modify) Bone - Bone: thigh_volume_r"(ComponentPose=@n5, Alpha=Get thigh_rot_r(), Translation="1,2,-1", Rotation="0,0,0", Scale="1,1,1") {
    translationMode=BMM_Additive
    translationSpace=BCS_BoneSpace
    alphaBoolBlend {blendOption=Linear}
    boneToModify {boneName=thigh_volume_r}
    translation=<1 2 -1>
   }
   n7 = "Transform (Modify) Bone - Bone: dyn_belt_base"(ComponentPose=@n6, Alpha=Get jump_alpha() -> (jump_alpha:double), Translation="0,0,0", Rotation="-20,0,0", Scale="1,1,1") {
    rotationMode=BMM_Additive
    rotationSpace=BCS_BoneSpace
    scaleMode=BMM_Additive
    scaleSpace=BCS_BoneSpace
    translationMode=BMM_Additive
    translationSpace=BCS_BoneSpace
    alphaBoolBlend {blendOption=Linear}
    alphaScaleBiasClamp {bInterpResult=true, inRange {max=0, min=-100}, outRange {max=0, min=1}}
    boneToModify {boneName=dyn_belt_base}
    rotation {pitch=-20}
   }
   n8 = Get thigh_rot_r()
   n9 = "Transform (Modify) Bone - Bone: dyn_belt_a_base"(ComponentPose=@n7, Alpha=@n8, Translation="-12,2,2", Rotation="-30,90,0", Scale="1,1,1") {
    rotationMode=BMM_Additive
    rotationSpace=BCS_BoneSpace
    scaleMode=BMM_Additive
    scaleSpace=BCS_BoneSpace
    translationMode=BMM_Additive
    translationSpace=BCS_BoneSpace
    alphaBoolBlend {blendOption=Linear}
    alphaScaleBiasClamp {bMapRange=true, inRange {max=0, min=-100}, outRange {max=0, min=1}}
    boneToModify {boneName=dyn_belt_a_base}
    rotation {pitch=-30, yaw=90}
    translation=<-12 2 2>
   }
   n10 = "Transform (Modify) Bone - Bone: dyn_belt_b_base"(ComponentPose=@n9, Alpha=@n8, Translation="-7,0,2", Rotation="20,0,0", Scale="1,1,1") {
    rotationMode=BMM_Additive
    rotationSpace=BCS_BoneSpace
    scaleMode=BMM_Additive
    scaleSpace=BCS_BoneSpace
    translationMode=BMM_Additive
    translationSpace=BCS_BoneSpace
    alphaBoolBlend {blendOption=Linear}
    alphaScaleBiasClamp {bMapRange=true, inRange {max=0, min=-100}, outRange {max=0, min=1}}
    boneToModify {boneName=dyn_belt_b_base}
    rotation {pitch=20}
    translation {x=-7, z=2}
   }
   n11 = "Transform (Modify) Bone - Bone: dyn_belt_c_base"(ComponentPose=@n10, Alpha=@n8, Translation="0,0,0", Rotation="0,0,0", Scale="1,1,1") {
    rotationMode=BMM_Additive
    rotationSpace=BCS_BoneSpace
    scaleMode=BMM_Additive
    scaleSpace=BCS_BoneSpace
    translationMode=BMM_Additive
    translationSpace=BCS_BoneSpace
    alphaBoolBlend {blendOption=Linear}
    alphaScaleBiasClamp {bMapRange=true, inRange {max=0, min=-100}, outRange {max=0, min=1}}
    boneToModify {boneName=dyn_belt_c_base}
   }
   n12 = Get thigh_side_r() -> (thigh_side_r:double)
   n13 = "Transform (Modify) Bone - Bone: dyn_belt_a_base"(ComponentPose=@n11, Alpha=@n12, Translation="3,0,0", Rotation="-20,50,0", Scale="1,1,1") {
    rotationMode=BMM_Additive
    rotationSpace=BCS_BoneSpace
    scaleMode=BMM_Additive
    scaleSpace=BCS_BoneSpace
    translationMode=BMM_Additive
    translationSpace=BCS_BoneSpace
    alphaBoolBlend {blendOption=Linear}
    alphaScaleBiasClamp {bMapRange=true, inRange {max=35, min=10}}
    boneToModify {boneName=dyn_belt_a_base}
    rotation {pitch=-20, yaw=50}
    translation {x=3}
   }
   n14 = "Transform (Modify) Bone - Bone: dyn_belt_b_base"(ComponentPose=@n13, Alpha=@n12, Translation="-3,0,0", Rotation="-20,50,0", Scale="1,1,1") {
    rotationMode=BMM_Additive
    rotationSpace=BCS_BoneSpace
    scaleMode=BMM_Additive
    scaleSpace=BCS_BoneSpace
    translationMode=BMM_Additive
    translationSpace=BCS_BoneSpace
    alphaBoolBlend {blendOption=Linear}
    alphaScaleBiasClamp {bMapRange=true, inRange {max=35, min=10}}
    boneToModify {boneName=dyn_belt_b_base}
    rotation {pitch=-20, yaw=50}
    translation {x=-3}
   }
   n15 = "Transform (Modify) Bone - Bone: dyn_belt_c_base"(ComponentPose=@n14, Alpha=@n12, Translation="0,0,0", Rotation="20,0,0", Scale="1,1,1") {
    rotationMode=BMM_Additive
    rotationSpace=BCS_BoneSpace
    scaleMode=BMM_Additive
    scaleSpace=BCS_BoneSpace
    translationMode=BMM_Additive
    translationSpace=BCS_BoneSpace
    alphaBoolBlend {blendOption=Linear}
    alphaScaleBiasClamp {bMapRange=true, inRange {max=35, min=10}}
    boneToModify {boneName=dyn_belt_c_base}
    rotation {pitch=20}
   }
   n16 = "Constraint - dyn_ammo_a (PPPPPP)"!(
    ComponentPose=@n15
    Alpha=1
    ConstraintWeights_0=0.0163
    ConstraintWeights_1:float=0.0031
    ConstraintWeights_2:float=0.4082
    ConstraintWeights_3:float=0.5724
    ConstraintWeights_4:float=0
    ConstraintWeights_5:float=0
   ) -> (Pose) {
    alphaBoolBlend {blendOption=Linear}
    boneToModify {boneName=dyn_ammo_a}
    constraintSetup [
     {targetBone {boneName=thigh_twist_01_r}}
     {targetBone {boneName=thigh_twist_01_l}}
     {targetBone {boneName=pelvis}}
     {targetBone {boneName=dyn_belt_a}}
     {targetBone {boneName=dyn_belt_b}}
     {targetBone {boneName=dyn_belt_c}}
    ]
    constraintWeights [1,1,1,1,1,1]
   }
   n17 = "Constraint - dyn_ammo_b (PPPPPP)"!(
    ComponentPose=@n16
    Alpha=1
    ConstraintWeights_0=0.0025
    ConstraintWeights_1:float=0.0004
    ConstraintWeights_2:float=0.0491
    ConstraintWeights_3:float=0.9475
    ConstraintWeights_4:float=0.0005
    ConstraintWeights_5:float=0
   ) -> (Pose) {
    alphaBoolBlend {blendOption=Linear}
    boneToModify {boneName=dyn_ammo_b}
    constraintSetup [
     {targetBone {boneName=thigh_twist_01_r}}
     {targetBone {boneName=thigh_twist_01_l}}
     {targetBone {boneName=pelvis}}
     {targetBone {boneName=dyn_belt_a}}
     {targetBone {boneName=dyn_belt_b}}
     {targetBone {boneName=dyn_belt_c}}
    ]
    constraintWeights [1,1,1,1,1,1]
   }
   n18 = "Constraint - dyn_ammo_c (PPPPPP)"!(
    ComponentPose=@n17
    Alpha=1
    ConstraintWeights_0=0
    ConstraintWeights_1:float=0
    ConstraintWeights_2:float=0.0009
    ConstraintWeights_3:float=0.9647
    ConstraintWeights_4:float=0.0344
    ConstraintWeights_5:float=0
   ) -> (Pose) {
    alphaBoolBlend {blendOption=Linear}
    boneToModify {boneName=dyn_ammo_c}
    constraintSetup [
     {targetBone {boneName=thigh_twist_01_r}}
     {targetBone {boneName=thigh_twist_01_l}}
     {targetBone {boneName=pelvis}}
     {targetBone {boneName=dyn_belt_a}}
     {targetBone {boneName=dyn_belt_b}}
     {targetBone {boneName=dyn_belt_c}}
    ]
    constraintWeights [1,1,1,1,1,1]
   }
   n19 = "Constraint - dyn_ammo_d (PPPPPP)"!(
    ComponentPose=@n18
    Alpha=1
    ConstraintWeights_0=0
    ConstraintWeights_1:float=0
    ConstraintWeights_2:float=0
    ConstraintWeights_3:float=0.6321
    ConstraintWeights_4:float=0.3678
    ConstraintWeights_5:float=0.0001
   ) -> (Pose) {
    alphaBoolBlend {blendOption=Linear}
    boneToModify {boneName=dyn_ammo_d}
    constraintSetup [
     {targetBone {boneName=thigh_twist_01_r}}
     {targetBone {boneName=thigh_twist_01_l}}
     {targetBone {boneName=pelvis}}
     {targetBone {boneName=dyn_belt_a}}
     {targetBone {boneName=dyn_belt_b}}
     {targetBone {boneName=dyn_belt_c}}
    ]
    constraintWeights [1,1,1,1,1,1]
   }
   n20 = "Constraint - dyn_ammo_e (PPPPPP)"!(
    ComponentPose=@n19
    Alpha=1
    ConstraintWeights_0=0
    ConstraintWeights_1:float=0
    ConstraintWeights_2:float=0
    ConstraintWeights_3:float=0.2114
    ConstraintWeights_4:float=0.7866
    ConstraintWeights_5:float=0.002
   ) -> (Pose) {
    alphaBoolBlend {blendOption=Linear}
    boneToModify {boneName=dyn_ammo_e}
    constraintSetup [
     {targetBone {boneName=thigh_twist_01_r}}
     {targetBone {boneName=thigh_twist_01_l}}
     {targetBone {boneName=pelvis}}
     {targetBone {boneName=dyn_belt_a}}
     {targetBone {boneName=dyn_belt_b}}
     {targetBone {boneName=dyn_belt_c}}
    ]
    constraintWeights [1,1,1,1,1,1]
   }
   n21 = "Constraint - dyn_ammo_f (PPPPPPPPP)"!(
    ComponentPose=@n20
    Alpha=1
    ConstraintWeights_0=0
    ConstraintWeights_1:float=0
    ConstraintWeights_2:float=0
    ConstraintWeights_3:float=0
    ConstraintWeights_4:float=0
    ConstraintWeights_5:float=0
    ConstraintWeights_6:float=0.0406
    ConstraintWeights_7:float=0.931
    ConstraintWeights_8:float=0.0284
   ) -> (Pose) {
    alphaBoolBlend {blendOption=Linear}
    boneToModify {boneName=dyn_ammo_f}
    constraintSetup [
     {targetBone {boneName=thigh_volume_l}}
     {targetBone {boneName=thigh_twist_01_r}}
     {targetBone {boneName=spine_01}}
     {targetBone {boneName=thigh_twist_01_l}}
     {targetBone {boneName=thigh_volume_r}}
     {targetBone {boneName=pelvis}}
     {targetBone {boneName=dyn_belt_a}}
     {targetBone {boneName=dyn_belt_b}}
     {targetBone {boneName=dyn_belt_c}}
    ]
    constraintWeights [1,1,1,1,1,1,1,1,1]
   }
   n22 = RigidBody(ComponentPose=@n21) {
    bClampLinearTranslationLimitToRefPose=true
    bOverrideWorldGravity=true
    overridePhysicsAsset=&/F_MED_Anti_Llama_Physics
    simulationSpace=BaseBoneSpace
    alphaBoolBlend {blendOption=Linear}
    baseBoneRef {boneName=pelvis}
    componentLinearAccScale=<0.3 0.3 0.3>
    componentLinearVelScale=<0.3 0.3 0.3>
    overrideWorldGravity {z=-980}
   }
   n23 = "Constraint - dyn_collar_bk_mstr (PPPP)"!(
    ComponentPose=@n22
    Alpha=1
    ConstraintWeights_0=0.2784
    ConstraintWeights_1:float=0.1214
    ConstraintWeights_2:float=0.2223
    ConstraintWeights_3:float=0.3779
   ) -> (Pose) {
    alphaBoolBlend {blendOption=Linear}
    boneToModify {boneName=dyn_collar_bk_mstr}
    constraintSetup [{targetBone {boneName=spine_05}}, {targetBone {boneName=neck_01}}, {targetBone {boneName=head}}, {targetBone {boneName=neck_02}}]
    constraintWeights [1,1,1,1]
   }
   n24 = "Constraint - dyn_collar_l_mstr (PPPPP)"!(
    ComponentPose=@n23
    Alpha=1
    ConstraintWeights_0=0.4346
    ConstraintWeights_1:float=0.0126
    ConstraintWeights_2:float=0.1616
    ConstraintWeights_3:float=0.0781
    ConstraintWeights_4:float=0.313
   ) -> (Pose) {
    alphaBoolBlend {blendOption=Linear}
    boneToModify {boneName=dyn_collar_l_mstr}
    constraintSetup [
     {targetBone {boneName=spine_05}}
     {targetBone {boneName=clavicle_l}}
     {targetBone {boneName=neck_01}}
     {targetBone {boneName=head}}
     {targetBone {boneName=neck_02}}
    ]
    constraintWeights [1,1,1,1,1]
   }
   n25 = "Constraint - dyn_collar_r_mstr (PPPPP)"!(
    ComponentPose=@n24
    Alpha=1
    ConstraintWeights_0=0.4285
    ConstraintWeights_1:float=0.0122
    ConstraintWeights_2:float=0.1562
    ConstraintWeights_3:float=0.1006
    ConstraintWeights_4:float=0.3025
   ) -> (Pose) {
    alphaBoolBlend {blendOption=Linear}
    boneToModify {boneName=dyn_collar_r_mstr}
    constraintSetup [
     {targetBone {boneName=spine_05}}
     {targetBone {boneName=clavicle_r}}
     {targetBone {boneName=neck_01}}
     {targetBone {boneName=head}}
     {targetBone {boneName=neck_02}}
    ]
    constraintWeights [1,1,1,1,1]
   }
   n26 = "Constraint - dyn_collar_fr_mstr (PPPP)"!(
    ComponentPose=@n25
    Alpha=1
    ConstraintWeights_0=0.4734
    ConstraintWeights_1:float=0.204
    ConstraintWeights_2:float=0.0171
    ConstraintWeights_3:float=0.3054
   ) -> (Pose) {
    alphaBoolBlend {blendOption=Linear}
    boneToModify {boneName=dyn_collar_fr_mstr}
    constraintSetup [{targetBone {boneName=spine_05}}, {targetBone {boneName=neck_01}}, {targetBone {boneName=head}}, {targetBone {boneName=neck_02}}]
    constraintWeights [1,1,1,1]
   }
   n27 = Control Rig(Source=Component To Local(ComponentPose=@n26)) {
    alphaBoolBlend {blendOption=Linear}
    controlRigAssetReference {blueprintRigClass=&/F_MED_Anti_Llama_CtrlRig.F_MED_Anti_Llama_CtrlRig_C}
   }
   Output Pose(Result=@n27) {layerGroup=None}
  ]}
 EventGraph {nodes [
   Event BlueprintUpdateAnimation()
   > n1 = Sequence!(execute) -> (then_0, then_1, then_2:Exec)
   > n2 = Sequence!(execute) -> (then_0, then_1)
   n3 = GetOwnerSkeletalMeshComponent()
   n4 = BreakRotator(InRot=BreakTransform(InTransform=GetDeltaTransformFromRefPose(self=@n3, BoneName=thigh_r, BaseName=pelvis)).Rotation)
   Set thigh_rot_r(execute=@n2.then_1, thigh_rot_r=@n4.Yaw)
   > Set thigh_side_r(thigh_side_r:double=@n4.Pitch) -> (Output_Get:double)
   n5 = GetDeltaTransformFromRefPose(self=GetOwnerSkeletalMeshComponent(), BoneName=thigh_l, BaseName=pelvis)
   Set thigh_rot_l(execute=@n2.then_0, thigh_rot_l=BreakRotator(InRot=BreakTransform(InTransform=@n5).Rotation).Yaw)
   Set LOD1(execute=@n1.then_2, LOD1:bool="integer >= integer"(A=Get CurrentLOD(), B=1)) -> (Output_Get:bool)
   n6 = GetOwnerSkeletalMeshComponent()
   n7 = BreakRotator(InRot=BreakTransform(InTransform=GetDeltaTransformFromRefPose(self=@n6, BoneName=neck_01, BaseName=spine_05)).Rotation)
   > Set look_up(look_up:double=@n7) -> (Output_Get:double)
   Set jump_alpha(execute=@n1.then_1, jump_alpha:double=MapRangeClamped(Value=Get PawnVelocityZ(), InRangeA=-700, InRangeB=0, OutRangeA=1, OutRangeB=0)) -> (Output_Get:double)
   MapRangeClamped(Value=@n7, InRangeA=50, InRangeB=65)
  ]}
]
variables [
 jump_alpha {type=number}
 LOD1 {type=boolean}
 look_up {type=number}
 thigh_rot_l {type=number}
 thigh_rot_r {type=number}
 thigh_side_r {type=number}
]
