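# Physics asset for the F_MED_Anti_Llama character, pairing upper-body collision capsules with bodies for dynamic hair bones.
# Spine, clavicle, upper arm, neck and head capsules are linked by fully free constraints, as are both ponytail bases to the head.
# Limited-angle constraints chain the asymmetric ponytails (longer on the right), front hair and left side hair.
# Both right side-hair bodies and the top tuft attach directly to the head with limited motion.
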PhysicsAsset /BRCosmetics/Characters/Player/Female/Medium/Bodies/F_MED_Anti_Llama/Meshes/Parts/F_MED_Anti_Llama_Physics
bodies {
 clavicle_l [{shapeType=Capsule, length=8, radius=4.5, center=<6.3469 -1.095 -0.5847>, rotation=<rot 80.1433 0 0>}]
 clavicle_r [{shapeType=Capsule, length=8, radius=4.5, center=<-6.3469 1.095 0.5847>, rotation=<rot 80.1433 0 -180>}]
 dyn_front_hair_a [{shapeType=Capsule, length=2.5, radius=1.5, center=<2.4963 -0.2917 0.2225>, rotation=<rot 90 -90 -83.1781>}]
 dyn_front_hair_b [{shapeType=Capsule, length=2, radius=1.75, center=<2.147 0 -0.2158>, rotation=<rot 80.9008 -0.0001 -0.0001>}]
 dyn_front_hair_c [{shapeType=Capsule, length=2.75, radius=1, center=<2.1817 0 0>, rotation=<rot 90 0 0>}]
 dyn_ponyTail_hair_a_base_l [{shapeType=Capsule, length=1, radius=1}]
 dyn_ponyTail_hair_a_base_r [{shapeType=Capsule, length=1, radius=1}]
 dyn_ponyTail_hair_a_l [{shapeType=Capsule, length=2.5, radius=2, center=<1.5786 0 0.4942>, rotation=<rot 89.3574 -179.9996 -179.9996>}]
 dyn_ponyTail_hair_a_r [{shapeType=Capsule, length=0.75, radius=2.5, center=<-2.1461 0.0001 -0.4535>, rotation=<rot 90 0 0>}]
 dyn_ponyTail_hair_b_l [{shapeType=Capsule, length=1.5, radius=1.75, center=<2.1821 -0.3579 -0.0997>, rotation=<rot 90 0 0>}]
 dyn_ponyTail_hair_b_r [{shapeType=Capsule, length=1, radius=2, center=<-2.1771 0.1212 0>, rotation=<rot 90 0 0>}]
 dyn_ponyTail_hair_c_l [{shapeType=Capsule, length=1.75, radius=1.25, center=<2.2344 -0.2541 0>, rotation=<rot 90 0 0>}]
 dyn_ponyTail_hair_c_r [{shapeType=Capsule, length=1, radius=1.75, center=<-2.1589 0.3327 -0.0744>, rotation=<rot 90 0 0>}]
 dyn_ponyTail_hair_d_l [{shapeType=Capsule, length=5, radius=1.5, center=<3.8342 0 0>, rotation=<rot 90 0 0>}]
 dyn_ponyTail_hair_d_r [{shapeType=Capsule, length=1.25, radius=1.75, center=<-2.1366 0.3473 0>, rotation=<rot 90 0 0>}]
 dyn_ponyTail_hair_e_r [{shapeType=Capsule, length=1.75, radius=1.25, center=<-1.8865 0 0>, rotation=<rot 90 0 0>}]
 dyn_ponyTail_hair_f_r [{shapeType=Capsule, length=3.75, radius=2.25, center=<-4.1368 0 0>, rotation=<rot 90 0 0>}]
 dyn_side_hair_a_l [{shapeType=Capsule, length=3, radius=1.75, center=<2.7931 0.0002 0>, rotation=<rot 90 0 0>}]
 dyn_side_hair_a_r [{shapeType=Capsule, length=2, radius=1, center=<0.0002 -1.4549 0>, rotation=<rot 0 0 90>}]
 dyn_side_hair_b_l [{shapeType=Capsule, length=4, radius=1.25, center=<2.9143 0 0>, rotation=<rot 90 0 0>}]
 dyn_side_hair_b_r [{shapeType=Capsule, length=2, radius=1, center=<0.0002 -1.4549 0>, rotation=<rot 0 0 90>}]
 dyn_top_hair_a [{shapeType=Capsule, length=1.25, radius=1.25, center=<0 1.2177 0>, rotation=<rot 0 0 90>}]
 head [{shapeType=Capsule, length=7, radius=5.75, center=<2.9286 2.6847 0>, rotation=<rot 90 0 0>}]
 neck_01 [{shapeType=Capsule, length=1.75, radius=4.5, center=<-1.1589 0.5279 0>, rotation=<rot 90 0 0>}]
 neck_02 [{shapeType=Capsule, length=1.5, radius=4.25, center=<-1.0205 0.3244 0>, rotation=<rot 90 0 0>}]
 spine_04 [{shapeType=Capsule, length=2.75, radius=10.5, center=<-0.9726 2.9073 0>}]
 spine_05 [{shapeType=Capsule, length=9, radius=8.25, center=<0.5433 -0.3614 0.0001>}]
 upperarm_l [{shapeType=Capsule, length=18.75, radius=3.75, center=<12.742 -0.5232 -0.5109>, rotation=<rot 89.6236 -0.8388 -0.0001>}]
 upperarm_r [{shapeType=Capsule, length=18.75, radius=3.75, center=<-12.742 0.5232 0.5109>, rotation=<rot 89.6231 -0.8388 179.9999>}]
}
constraints [
 {
  bone1Name=spine_05
  bone2Name=spine_04
  swing1LimitDegrees=45
  swing1Motion=Free
  swing2LimitDegrees=45
  swing2Motion=Free
  twistLimitDegrees=45
  twistMotion=Free
 }
 {
  bone1Name=clavicle_l
  bone2Name=spine_05
  swing1LimitDegrees=45
  swing1Motion=Free
  swing2LimitDegrees=45
  swing2Motion=Free
  twistLimitDegrees=45
  twistMotion=Free
 }
 {
  bone1Name=upperarm_l
  bone2Name=clavicle_l
  swing1LimitDegrees=45
  swing1Motion=Free
  swing2LimitDegrees=45
  swing2Motion=Free
  twistLimitDegrees=45
  twistMotion=Free
 }
 {
  bone1Name=neck_01
  bone2Name=spine_05
  swing1LimitDegrees=45
  swing1Motion=Free
  swing2LimitDegrees=45
  swing2Motion=Free
  twistLimitDegrees=45
  twistMotion=Free
 }
 {
  bone1Name=neck_02
  bone2Name=neck_01
  swing1LimitDegrees=45
  swing1Motion=Free
  swing2LimitDegrees=45
  swing2Motion=Free
  twistLimitDegrees=45
  twistMotion=Free
 }
 {
  bone1Name=head
  bone2Name=neck_02
  swing1LimitDegrees=45
  swing1Motion=Free
  swing2LimitDegrees=45
  swing2Motion=Free
  twistLimitDegrees=45
  twistMotion=Free
 }
 {
  bone1Name=dyn_ponyTail_hair_a_base_l
  bone2Name=head
  swing1LimitDegrees=45
  swing1Motion=Free
  swing2LimitDegrees=45
  swing2Motion=Free
  twistLimitDegrees=45
  twistMotion=Free
 }
 {
  bone1Name=dyn_ponyTail_hair_a_base_r
  bone2Name=head
  swing1LimitDegrees=45
  swing1Motion=Free
  swing2LimitDegrees=45
  swing2Motion=Free
  twistLimitDegrees=45
  twistMotion=Free
 }
 {bone1Name=dyn_ponyTail_hair_a_l, bone2Name=dyn_ponyTail_hair_a_base_l, swing1LimitDegrees=20, swing2LimitDegrees=20, twistLimitDegrees=5}
 {bone1Name=dyn_ponyTail_hair_b_l, bone2Name=dyn_ponyTail_hair_a_l, swing1LimitDegrees=20, swing2LimitDegrees=20, twistLimitDegrees=5}
 {bone1Name=dyn_ponyTail_hair_c_l, bone2Name=dyn_ponyTail_hair_b_l, swing1LimitDegrees=25, swing2LimitDegrees=25, twistLimitDegrees=5}
 {bone1Name=dyn_ponyTail_hair_d_l, bone2Name=dyn_ponyTail_hair_c_l, swing1LimitDegrees=25, swing2LimitDegrees=25, twistLimitDegrees=5}
 {bone1Name=dyn_ponyTail_hair_a_r, bone2Name=dyn_ponyTail_hair_a_base_r, swing1LimitDegrees=20, swing2LimitDegrees=20, twistLimitDegrees=5}
 {bone1Name=dyn_ponyTail_hair_b_r, bone2Name=dyn_ponyTail_hair_a_r, swing1LimitDegrees=20, swing2LimitDegrees=20, twistLimitDegrees=5}
 {bone1Name=dyn_ponyTail_hair_c_r, bone2Name=dyn_ponyTail_hair_b_r, swing1LimitDegrees=25, swing2LimitDegrees=25, twistLimitDegrees=5}
 {bone1Name=dyn_ponyTail_hair_d_r, bone2Name=dyn_ponyTail_hair_c_r, swing1LimitDegrees=25, swing2LimitDegrees=25, twistLimitDegrees=5}
 {bone1Name=dyn_ponyTail_hair_e_r, bone2Name=dyn_ponyTail_hair_d_r, swing1LimitDegrees=25, swing2LimitDegrees=25, twistLimitDegrees=5}
 {bone1Name=dyn_ponyTail_hair_f_r, bone2Name=dyn_ponyTail_hair_e_r, swing1LimitDegrees=25, swing2LimitDegrees=25, twistLimitDegrees=5}
 {
  bone1Name=upperarm_r
  bone2Name=clavicle_r
  swing1LimitDegrees=45
  swing1Motion=Free
  swing2LimitDegrees=45
  swing2Motion=Free
  twistLimitDegrees=45
  twistMotion=Free
 }
 {
  bone1Name=clavicle_r
  bone2Name=spine_05
  swing1LimitDegrees=45
  swing1Motion=Free
  swing2LimitDegrees=45
  swing2Motion=Free
  twistLimitDegrees=45
  twistMotion=Free
 }
 {bone1Name=dyn_side_hair_a_l, bone2Name=head, swing1LimitDegrees=20, swing2LimitDegrees=20, twistLimitDegrees=5}
 {bone1Name=dyn_side_hair_b_l, bone2Name=dyn_side_hair_a_l, swing1LimitDegrees=20, swing2LimitDegrees=20, twistLimitDegrees=5}
 {bone1Name=dyn_front_hair_a, bone2Name=head, swing1LimitDegrees=10, swing2LimitDegrees=10, twistLimitDegrees=5}
 {bone1Name=dyn_front_hair_b, bone2Name=dyn_front_hair_a, swing1LimitDegrees=15, swing2LimitDegrees=15, twistLimitDegrees=5}
 {bone1Name=dyn_front_hair_c, bone2Name=dyn_front_hair_b, swing1LimitDegrees=15, swing2LimitDegrees=15, twistLimitDegrees=5}
 {bone1Name=dyn_side_hair_a_r, bone2Name=head, swing1LimitDegrees=15, swing2LimitDegrees=15, twistLimitDegrees=15}
 {bone1Name=dyn_side_hair_b_r, bone2Name=head, swing1LimitDegrees=15, swing2LimitDegrees=15, twistLimitDegrees=15}
 {bone1Name=dyn_top_hair_a, bone2Name=head, swing1LimitDegrees=10, swing2LimitDegrees=10, twistLimitDegrees=10}
]
